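# Physics asset for the Pink Widow backpack cosmetic, giving the pack and its dangling star, strap and kunai pieces capsule bodies for secondary motion.
# A small root pack body anchors dyn_pack through tight angular limits, and dyn_pack carries two stars and a four-link strap chain ending in the kunai.
# Twist is locked on the last strap link and the kunai, whose second swing axis is also locked, so it swings like a hinge.
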PhysicsAsset /BRCosmetics/Accessories/FORT_Backpacks/Backpack_F_MED_Pink_Widow/Meshes/F_MED_Pink_Widow_Pack_Physics
bodies {
 dyn_kunai [{shapeType=Capsule, length=7.5988, radius=1.9439, shapeName=dyn_kunai_capsule, center=<7.4308 0 0>, rotation=<rot -90 180 180>}]
 dyn_pack [{shapeType=Capsule, length=25, radius=3.5, shapeName=dyn_pack_capsule, center=<2.9182 5.8021 -1.9191>, rotation=<rot 0 0.244 -62.8746>}]
 dyn_star1 [{shapeType=Capsule, length=3.9889, radius=1.7201, shapeName=dyn_star1_capsule, center=<1.2666 0 0>, rotation=<rot 0 0 -90>}]
 dyn_star2 [{shapeType=Capsule, length=4, radius=1.7, shapeName=dyn_star2_capsule, center=<1.34 0 0>, rotation=<rot 0 0 -90>}]
 dyn_strap1 [{shapeType=Capsule, length=0.6812, radius=1.0664, shapeName=dyn_strap1_capsule, center=<0.9378 0 0>, rotation=<rot -90 180 180>}]
 dyn_strap2 [{shapeType=Capsule, length=0.7183, radius=1.0125, shapeName=dyn_strap2_capsule, center=<1.1509 0 0>, rotation=<rot -90 0 0>}]
 dyn_strap3 [{shapeType=Capsule, length=1.0794, radius=0.9874, shapeName=dyn_strap3_capsule, center=<1.1424 0 0>, rotation=<rot -90 180 180>}]
 dyn_strap4 [{shapeType=Capsule, length=0.2664, radius=0.9898, shapeName=dyn_strap4_capsule, center=<0.621 0 0>, rotation=<rot -90 0 0>}]
 pack [{shapeType=Capsule, length=0.505, radius=0.505, shapeName=pack_capsule, rotation=<rot 0 -90 0>}]
}
constraints [
 {bone1Name=dyn_star1, bone2Name=dyn_pack, swing1LimitDegrees=6, swing2LimitDegrees=10, twistLimitDegrees=6}
 {bone1Name=dyn_star2, bone2Name=dyn_pack, swing1LimitDegrees=6, swing2LimitDegrees=10, twistLimitDegrees=6}
 {bone1Name=dyn_strap1, bone2Name=dyn_pack, swing1LimitDegrees=12, swing2LimitDegrees=12, twistLimitDegrees=1}
 {bone1Name=dyn_strap2, bone2Name=dyn_strap1, swing1LimitDegrees=12, swing2LimitDegrees=12, twistLimitDegrees=1}
 {bone1Name=dyn_strap3, bone2Name=dyn_strap2, swing1LimitDegrees=12, swing2LimitDegrees=12, twistLimitDegrees=1}
 {bone1Name=dyn_strap4, bone2Name=dyn_strap3, swing1LimitDegrees=12, swing2LimitDegrees=12, twistLimitDegrees=6, twistMotion=Locked}
 {
  bone1Name=dyn_kunai
  bone2Name=dyn_strap4
  swing1LimitDegrees=12
  swing2LimitDegrees=45
  swing2Motion=Locked
  twistLimitDegrees=45
  twistMotion=Locked
 }
 {bone1Name=dyn_pack, bone2Name=pack, swing1LimitDegrees=4, swing2LimitDegrees=4, twistLimitDegrees=8}
]
